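# Material function that outputs glitch-distorted UVs for a holographic screen effect, filed under the Dark Nexus quest prop.
# Static switches layer grid distortion, scrolling noise stretch, horizontal jitter and a sine wiggle onto the input UVs.
# A hashed-time sine, thresholded and multiplied by random flicker, forms a mask for intermittent bursts.
# Horizontal jitter combines a scrolling glitch map with a second sample stepped by floored time, offsetting only along U.
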MaterialFunction /Hera/Props/Quest/Quest_DarkNexusShape_A/Material/MF_ScreenGlitch_SpacePlunge
graph {nodes [
  n1 = NamedRerouteDeclaration(None=FunctionInput(Preview=TextureCoordinate()) {bUsePreviewValueAsDefault=true, inputName=Main_TextureUV, inputType=FunctionInput_Vector2, sortPriority=2}) {name=MainTexture_UV, nodeColor=<rgba 0.3828 0 1 1>}
  n2 = StaticSwitchParameter(
   True=LinearInterpolate(
    A=@n1
    B=MaterialFunctionCall!(UV=@n1) {materialFunction=/Game/Global/Materials/MaterialGym/Materials/HolographicTexture/MF_HolographicTextureGlitch_Grid}.Glitched UV
    Alpha=NamedRerouteUsage()
   )
   False=NamedRerouteUsage()
  ) {defaultValue=true, group=HolographicTexture_Glitch, parameterName=UseGridGlitch, sortPriority=0}
  n3 = StaticSwitchParameter(True=Add(A=@n2, B=NamedRerouteUsage()), False=@n2) {defaultValue=true, group=HolographicTexture_Glitch, parameterName=UseStretchedGlitch}
  n4 = StaticSwitchParameter(True=Add(A=@n3, B=NamedRerouteUsage()), False=@n3) {defaultValue=true, group=HolographicTexture_Glitch, parameterName=UseHoriztonalGlitch}
  n5 = Sine(None=Multiply(A=Time()) {constB=10})
  n6 = Constant() {desc=Wiggle Intensity, r=0.0075}
  n7 = StaticSwitchParameter(True=Multiply(A=NamedRerouteUsage(), B=Multiply(A=@n5, B=@n6) {constB=0.007}), False=Constant()) {defaultValue=true, group=HolographicTexture_Glitch, parameterName=UseWiggleGlitch}
  FunctionOutput(None=StaticSwitchParameter(True=Add(A=@n4, B=@n7), False=NamedRerouteUsage()) {defaultValue=true, group=HolographicTexture_Glitch, parameterName=UseGlitch, sortPriority=-1})
  n8 = Constant() {r=1}
  n9 = Subtract(A=@n8, B=ScalarParameter() {defaultValue=0.075, group=HolographicTexture_Glitch, parameterName=GlitchDuration, sortPriority=10})
  n10 = Multiply(A=Time(), B=ScalarParameter() {defaultValue=0.2, group=HolographicTexture_Glitch, parameterName=GlitchFrequency, sortPriority=10}) {constB=0.2}
  n11 = Add(
   A=@n10
   B=Multiply(A=MaterialFunctionCall(Float=@n10) {materialFunction=&MF_Hash}.Float) {constB=0.5}
  )
  n12 = Add(A=Multiply(A=Sine(None=@n11)) {constB=0.5}) {constB=0.5}
  NamedRerouteDeclaration(None=Multiply(
    A=SmoothStep(Min=@n9, Max=@n8, Value=@n12) {constMax=0.9, constMin=0.8}
    B=MaterialFunctionCall() {materialFunction=/Game/MF__FloatFlickerRandom_Generic}.RandomFlickerFloat
   )) {name=GlitchFrequencyMask, nodeColor=<rgba 0.9377 1 0.8953 1>}
  n13 = NamedRerouteUsage()
  n14 = ScalarParameter() {group=HolographicTexture_Glitch, parameterName=GlitchUVRotation}
  n15 = Panner(
   Coordinate=Multiply(A=Rotator(Coordinate=@n13, Time=@n14), B=Constant2Vector() {g=10})
   Speed=AppendVector(A=Constant() {r=0.5}, B=ScalarParameter() {defaultValue=5, group=HolographicTexture_Glitch, parameterName=StretchedScrollingSpeed})
  ) {speedX=0.5, speedY=5}
  n16 = ScalarParameter() {defaultValue=0.1, group=HolographicTexture_Glitch, parameterName=StretchedGlitchIntensity}
  n17 = NamedRerouteUsage()
  NamedRerouteDeclaration(None=Multiply(
    A=TextureSample(UVs=@n15) {samplerType=SAMPLERTYPE_Masks, texture=/Game/Characters/Player/Male/Medium/Bodies/M_MED_Storm_Soldier/Textures/ElectricityNoise_M}.R
    B=LinearInterpolate(B=@n16, Alpha=@n17) {constB=0.1}
   ) {constB=0.2}) {name=StretchedGlitch, nodeColor=<rgba 0.7109 0 1 1>}
  n18 = Constant2Vector() {r=1}
  n19 = Multiply(
   A=Rotator(Coordinate=NamedRerouteUsage(), Time=ScalarParameter() {group=HolographicTexture_Glitch, parameterName=GlitchUVRotation})
   B=Constant2Vector() {g=10}
  )
  n20 = AppendVector(A=Constant() {r=0.5}, B=ScalarParameter() {defaultValue=5, group=HolographicTexture_Glitch, parameterName=StretchedScrollingSpeed})
  n21 = TextureSample(UVs=Multiply(A=Constant2Vector() {g=0.1, r=1}, B=Panner(Coordinate=@n19, Speed=@n20) {speedX=0.5, speedY=5})) {samplerType=SAMPLERTYPE_LinearColor, texture=/Game/Environments/Asteria/Sets/NeonCity/Techart/Holograms/GlitchMap_M}
  n22 = Multiply(A=NamedRerouteUsage(), B=Constant2Vector() {g=7, r=1}) {constB=5}
  n23 = Time()
  n24 = Divide(A=Floor(None=Multiply(A=@n23) {constB=3})) {constB=3}
  n25 = Add(A=TextureSample(UVs=Add(A=@n22, B=@n24)) {samplerType=SAMPLERTYPE_LinearColor, texture=&GlitchMap_M}.G) {constB=0.3}
  n26 = Multiply(
   A=Multiply(A=SmoothStep(Value=@n21.R) {constMax=0.1}, B=Saturate(None=@n25))
   B=LinearInterpolate(B=ScalarParameter() {defaultValue=0.1, group=HolographicTexture_Glitch, parameterName=StretchedGlitchIntensity}, Alpha=NamedRerouteUsage()) {constB=0.1}
  ) {constB=0.2}
  n27 = AppendVector(
   A=Multiply(B=ScalarParameter() {defaultValue=0.1, group=HolographicTexture_Glitch, parameterName=StretchedGlitchIntensity}) {constA=-0.75}
   B=Constant()
  )
  n28 = NamedRerouteUsage()
  NamedRerouteDeclaration(None=Add(A=Multiply(A=@n18, B=@n26), B=LinearInterpolate(B=@n27, Alpha=@n28) {constB=0.1})) {name=HoriztonalGlitch, nodeColor=<rgba 1 0.2812 0 1>}
 ]}
