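# Material function that gives a Dark Nexus quest cube a "space plunge" look, blended over input material attributes by a mask, with a separate Star output.
# Interior-cubemap UVs fake a recessed room with depth and corner shading, alongside parallax sparkles, bump-offset runes and an analytic sphere glow.
# An optional HuntGlitch switch adds scanlines and noise opacity.
# Downward-facing light tints emissive, while world position offset collapses vertices toward CubeCenter to hide the cube.
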MaterialFunction /Hera/Props/Quest/Quest_DarkNexusShape_A/Material/MF_SpacePlungeCube
graph {nodes [
  n1 = FunctionInput() {inputType=FunctionInput_MaterialAttributes}
  n2 = VectorParameter() {group=Cube_Additional, parameterName=TilingUVW, defaultValue=<rgba 1 1 2 2>}
  n3 = MaterialFunctionCall!(
   UVs=TextureCoordinate() {coordinateIndex=1}
   View
   RoomZscale=ScalarParameter() {defaultValue=0.6, group=Cube_Additional, parameterName=ZRoomScale}
   Tiling=ComponentMask(None=@n2.RGB) {g=true, r=true}
   Randomize Rotation
  ) {materialFunction=/Game/Global/MaterialFunctions/MF_InteriorCubemapUV}
  n4 = Saturate(None=Divide(A=ComponentMask(None=@n3.UVW) {r=true}, B=ScalarParameter() {defaultValue=10, group=Cube_Additional, parameterName=ZUnit}))
  n5 = OneMinus(None=Power(Base=OneMinus(None=@n4), Exp=ScalarParameter() {defaultValue=3, group=Cube_Additional, parameterName=ZPow}))
  n6 = LinearInterpolate(A=ScalarParameter() {defaultValue=0.15, group=Cube_Additional, parameterName=MinGradientColor}, Alpha=@n5) {constA=0.2}
  n7 = Abs(None=@n3.UVW)
  n8 = Multiply(A=ComponentMask(None=@n7) {g=true}, B=ComponentMask(None=@n7) {b=true})
  n9 = LinearInterpolate(
   A=ScalarParameter() {defaultValue=0.8, group=Cube_Additional, parameterName=MinCornerColor}
   Alpha=Power(Base=OneMinus(None=@n8), Exp=ScalarParameter() {defaultValue=0.5, group=Cube_Additional, parameterName=CornerPow})
  )
  n10 = MaterialFunctionCall!(Blend Range=ScalarParameter() {defaultValue=150, group=Cube_OG, parameterName=StarEmissiveRange}, Start Offset=Constant()) {materialFunction=/Engine/Functions/Engine_MaterialFunctions03/Particles/Distance_Blend}
  n11 = Multiply(
   A=NamedRerouteUsage()
   B=LinearInterpolate(B=ScalarParameter() {defaultValue=200, group=Cube_OG, parameterName=StarFarIntensity}, Alpha=Power(Base=@n10.Result)) {constA=1}
  )
  n12 = NamedRerouteUsage()
  n13 = SetMaterialAttributes!(
   MaterialAttributes=@n1
   BaseColor=NamedRerouteUsage()
   Metallic=NamedRerouteUsage()
   Specular=NamedRerouteUsage()
   Roughness=NamedRerouteUsage()
   EmissiveColor=Multiply(A=Multiply(A=@n6, B=@n9), B=Add(A=@n12, B=@n11))
  ) {attributeSetTypes [
    #1
    #2
    #3
    #4
    #5
   ]}
  n14 = GetMaterialAttributes(None=@n13) {attributeGetTypes [#5]}
  n15 = VectorParameter() {group=Cube_Additional, parameterName=ScanLineColor, defaultValue=<rgba 0.224 0.0371 0 1>}
  n16 = Multiply(A=@n15.RGB, B=@n15.RGB)
  n17 = MaterialFunctionCall!(PreskinnedPOS, Scanline Amount=ScalarParameter() {defaultValue=50, group=Stablized, parameterName=ScanlineAmount}) {materialFunction=&/MF_Scanlines}
  n18 = NamedRerouteUsage()
  n19 = Add(A=@n14.EmissiveColor, B=Multiply(A=Multiply(A=@n16, B=@n17.Result), B=LinearInterpolate(Alpha=@n18) {constA=0.5}))
  n20 = LinearInterpolate(A=@n17.SpaceNoise, Alpha=NamedRerouteUsage())
  n21 = GetMaterialAttributes(None=StaticSwitchParameter(
    True=SetMaterialAttributes!(MaterialAttributes=@n14.MaterialAttributes, EmissiveColor=@n19, OpacityMask=@n20) {attributeSetTypes [#5, #6]}
    False=@n13
   ) {group=Cube_Additional_Control, parameterName=UseHuntGlitch}) {attributeGetTypes [#5]}
  n22 = PixelNormalWS()
  n23 = Constant3Vector() {constant=<rgba 0 0 -1 0>}
  n24 = Power(Base=Saturate(None=DotProduct(A=@n22, B=@n23))) {constExponent=2.2}
  n25 = VectorParameter() {group=Cube_Additional, parameterName=LightColor01, defaultValue=<rgba 1 0.4029 0 1>}
  n26 = Multiply(A=@n25.RGB, B=@n25.RGB)
  n27 = Add(
   A=Multiply(A=@n21.EmissiveColor, B=ScalarParameter() {defaultValue=1, group=Cube_Additional, parameterName=OverallEmissive})
   B=Multiply(A=Multiply(A=@n24, B=@n26), B=ScalarParameter() {group=Cube_Additional, parameterName=LightIntensity01})
  )
  n28 = Multiply(
   A=Subtract(A=VectorParameter() {parameterName=CubeCenter, defaultValue=<rgba 0 0 0 1>}.RGB, B=WorldPosition().XYZ)
   B=Saturate(None=ScalarParameter() {group=Cube_Additional, parameterName=CubeHideCurve})
  )
  n29 = FunctionInput() {bUsePreviewValueAsDefault=true, inputName=Mask, inputType=FunctionInput_Scalar, sortPriority=1, previewValue=<1 0 0 1>}
  FunctionOutput(None=BlendMaterialAttributes(
    A=NamedRerouteUsage()
    B=SetMaterialAttributes!(MaterialAttributes=@n21.MaterialAttributes, EmissiveColor=@n27, WorldPositionOffset=@n28) {attributeSetTypes [#5, #7]}
    Alpha=@n29
   ))
  FunctionOutput(None=@n11) {outputName=Star, sortPriority=10}
  n30 = TextureCoordinate() {coordinateIndex=1}
  n31 = Constant2Vector() {g=0.5, r=0.5}
  n32 = MaterialFunctionCall(Float2=Abs(None=Subtract(A=@n30, B=@n31))) {materialFunction=&BreakOutFloat2Components}
  n33 = Max(A=@n32.R, B=@n32.R)
  Multiply(A=ScalarParameter() {defaultValue=1}, B=@n33)
  NamedRerouteDeclaration(None=Multiply(A=ComponentMask(None=@n3.UVW) {b=true, g=true}, B=ComponentMask(None=@n2.RGBA) {a=true, b=true})) {name=Texcoord1_Mod, nodeColor=<rgba 0 1 0.4141 1>}
  n34 = Saturate(None=Multiply(A=OneMinus(None=@n33), B=Constant() {r=1.7}))
  n35 = Constant() {r=30}
  n36 = Add(
   A=VectorParameter() {group=Cube_OG, parameterName=BaseDiffuse, defaultValue=<rgba 0.015 0.0013 0.1 1>}.RGB
   B=Multiply(A=NamedRerouteUsage(), B=Power(Base=@n34, Exp=@n35))
  )
  n37 = Constant() {r=1.95}
  n38 = Power(Base=Saturate(None=Multiply(A=@n33, B=@n37)), Exp=Constant() {r=5})
  n39 = NamedRerouteUsage()
  NamedRerouteDeclaration(None=LinearInterpolate(
    A=Saturate(None=@n36)
    B=VectorParameter() {group=Cube_OG, parameterName=EdgeDiffuse, defaultValue=<rgba 0.006 0.0005 0.04 1>}.RGB
    Alpha=MaterialFunctionCall(Transition Phase=@n38, Height Texture=@n39) {materialFunction=&HeightLerp}.Lerp Alpha No Contrast
   )) {name=Diffuse, nodeColor=<rgba 0 1 0.5078 1>}
  n40 = Constant() {r=1.4}
  n41 = Power(Base=Saturate(None=Multiply(A=@n33, B=@n40)), Exp=Constant() {r=1.5})
  NamedRerouteDeclaration(None=LinearInterpolate(
    A=ScalarParameter() {defaultValue=0.29, group=Cube_OG, parameterName=Rough_Center}
    B=ScalarParameter() {defaultValue=0.5, group=Cube_OG, parameterName=Rough_Edge}
    Alpha=MaterialFunctionCall(Transition Phase=@n41, Height Texture=@n39) {materialFunction=&HeightLerp}.Lerp Alpha No Contrast
   )) {name=CubeRoughness, nodeColor=<rgba 0.6641 0 1 1>}
  NamedRerouteDeclaration(None=MaterialFunctionCall(Blend Range=ScalarParameter() {defaultValue=400, group=Cube_OG, parameterName=StarEmissiveRange02}, Start Offset=Constant()) {materialFunction=&Distance_Blend}.Result) {name=DistanceLerp, nodeColor=<rgba 0 0.25 1 1>}
  NamedRerouteDeclaration(None=@n5) {name=OuterRadiusMask, nodeColor=<rgba 0.3828 1 0 1>}
  n42 = VectorParameter() {group=Cube_OG, parameterName=GlowColor1, defaultValue=<rgba 0.012 0.0011 0.08 1>}
  n43 = VectorParameter() {group=Cube_OG, parameterName=GlowColor2, defaultValue=<rgba 0.0356 0.001 0.2 1>}
  n44 = VectorParameter() {group=Cube_OG, parameterName=GlowColor3, defaultValue=<rgba 0.1264 0 0.3 1>}
  n45 = VectorParameter() {group=Cube_OG, parameterName=GlowColor4, defaultValue=<rgba 0.3583 0 0.5 1>}
  n46 = Multiply(A=@n45.RGB, B=@n45.RGB)
  n47 = NamedRerouteUsage()
  n48 = Multiply(
   A=LinearInterpolate(A=Constant() {r=1200}, B=Constant() {r=1400}, Alpha=@n47)
   B=ScalarParameter() {defaultValue=1, parameterName=RadiusMult}
  )
  n49 = Multiply(A=@n48, B=@n48) {desc="r^2"}
  n50 = ObjectPositionWS()
  n51 = Add(A=VectorParameter() {group=Cube_OG, parameterName=offz, defaultValue=<rgba 0 0 -400 1>}.RGB, B=@n50)
  n52 = NamedRerouteUsage()
  n53 = Subtract(
   A=CameraPositionWS()
   B=Add(
    A=VectorParameter() {desc=For zapping players, group=Cube_OG, parameterName=SphereOffset}.RGB
    B=LinearInterpolate(A=@n50, B=@n51, Alpha=@n52)
   )
  )
  n54 = Distance(A=@n53, B=Constant())
  n55 = DotProduct(A=@n53, B=CameraVectorWS())
  n56 = Subtract(A=Multiply(A=@n54, B=@n54), B=Multiply(A=@n55, B=@n55))
  n57 = SquareRoot(None=Max(A=Subtract(A=@n49, B=@n56)) {constB=0})
  n58 = Multiply(A=Divide(A=Multiply(A=@n57) {constB=2}, B=Multiply(A=@n48) {constB=2}), B=ScalarParameter() {defaultValue=0.87, parameterName=gr334})
  n59 = SphereMask(A=TextureCoordinate() {coordinateIndex=1}, B=Constant() {r=0.5}) {attenuationRadius=0.7, hardnessPercent=30}
  n60 = Saturate(None=Add(A=NamedRerouteUsage(), B=ScalarParameter() {defaultValue=0.25, group=Cube_OG, parameterName=BaseCubeBackground}))
  n61 = ScalarParameter() {defaultValue=2.75, parameterName=2222}
  n62 = ScalarParameter() {defaultValue=1.5, parameterName=434233}
  n63 = Multiply(A=@n60, B=Power(Base=NamedRerouteUsage(), Exp=LinearInterpolate(A=@n61, B=@n62, Alpha=@n47)))
  n64 = MaterialFunctionCall(Transition Phase=@n58, Height Texture=Multiply(A=@n59, B=QualitySwitch(Default=@n63, Low=@n60, High=@n63))) {materialFunction=&HeightLerp}
  n65 = Multiply(
   A=Power(Base=@n58, Exp=ScalarParameter() {defaultValue=20, parameterName=gpow3})
   B=ScalarParameter() {defaultValue=75, parameterName=ActiveHotspot}
  )
  n66 = Power(Base=NamedRerouteUsage(), Exp=Constant() {r=3})
  n67 = Multiply(A=@n65, B=@n66)
  n68 = LinearInterpolate(
   A=@n65
   B=QualitySwitch(Default=@n67, Low=@n65, High=@n67)
   Alpha=ScalarParameter() {defaultValue=1, group=Cube_OG, parameterName=Blend-Tex-Sphere}
  )
  n69 = LinearInterpolate(A=Multiply(A=@n59, B=@n68), B=@n68, Alpha=ScalarParameter() {group=Cube_OG, parameterName=RemoveEdgeFromInnerGlow})
  n70 = Multiply(
   A=Add(
    A=MaterialFunctionCall(1=@n42.RGB, 2=@n43.RGB, 3=@n44.RGB, 4=@n46, A=@n64.Lerp Alpha No Contrast) {materialFunction=&Lerp_Multiple_Float3}.Lerp 4 Inputs
    B=Multiply(A=@n46, B=@n69)
   )
   B=@n59
  )
  n71 = ScalarParameter() {defaultValue=0.25, group=Cube_OG, parameterName=deactivated_1}
  n72 = ScalarParameter() {defaultValue=0.05, group=Cube_OG, parameterName=activated_1}
  n73 = NamedRerouteUsage()
  n74 = Saturate(None=Add(A=LinearInterpolate(A=@n71, B=@n72, Alpha=@n73), B=@n66))
  n75 = Multiply(A=@n59, B=@n64.Lerp Alpha No Contrast)
  n76 = ScalarParameter() {defaultValue=2.5, group=Cube_OG, parameterName=gpow3_1}
  n77 = VectorParameter() {group=Cube_OG, parameterName=GlowColor-Runes, defaultValue=<rgba 0.6076 0 1 1>}
  n78 = ScalarParameter() {defaultValue=75, group=Cube_OG, parameterName=RuneTexTiling04}
  n79 = TextureCoordinate() {uTiling=0.95, vTiling=0.75}
  n80 = MaterialFunctionCall(Float2=Add(A=Multiply(A=@n78, B=@n79), B=NamedRerouteUsage())) {materialFunction=&BreakOutFloat2Components}
  n81 = MaterialFunctionCall(X=Add(A=ScalarParameter() {parameterName=BB_19}, B=@n80.R), Y=Add(A=@n80.G, B=ScalarParameter() {defaultValue=0.5, parameterName=BB_20})) {materialFunction=&MakeFloat2}
  n82 = TextureSampleParameter2D(UVs=@n81.Result) {
   group=Cube_OG
   parameterName=RuneTex
   samplerSource=SSM_Wrap_WorldGroupSettings
   samplerType=SAMPLERTYPE_LinearColor
   texture=&/DFHeightMask
  }
  n83 = LinearInterpolate(A=Constant() {r=-1}, B=Constant() {r=1}, Alpha=@n82.G) {constA=-1}
  n84 = ScalarParameter() {defaultValue=1, group=Cube_OG, parameterName=Rune_Intensity01}
  n85 = TextureSampleParameter2D(UVs=BumpOffset(
    Coordinate=@n81.Result
    Height=Multiply(A=@n83, B=@n84)
    HeightRatioInput=ScalarParameter() {defaultValue=0.06, group=Cube_OG, parameterName=Rune_HeightRatio01}
   )) {
   group=Cube_OG
   parameterName=RuneTex
   samplerSource=SSM_Wrap_WorldGroupSettings
   samplerType=SAMPLERTYPE_LinearColor
   texture=&DFHeightMask
  }
  n86 = ScalarParameter() {defaultValue=10, parameterName=G_8}
  n87 = ScalarParameter() {defaultValue=2, parameterName=G_7}
  n88 = NamedRerouteUsage()
  n89 = ScalarParameter() {defaultValue=19, parameterName=RuneTexTiling03}
  n90 = TextureCoordinate() {vTiling=0.75}
  n91 = MaterialFunctionCall(Float2=Add(A=Multiply(A=@n89, B=@n90), B=NamedRerouteUsage())) {materialFunction=&BreakOutFloat2Components}
  n92 = MaterialFunctionCall(
   X=Add(A=ScalarParameter() {defaultValue=0.1, parameterName=BB_13}, B=@n91.R)
   Y=Add(A=@n91.G, B=ScalarParameter() {defaultValue=0.05, parameterName=BB_14})
  ) {materialFunction=&MakeFloat2}
  n93 = TextureSampleParameter2D(UVs=@n92.Result) {
   group=Cube_OG
   parameterName=RuneTex
   samplerSource=SSM_Wrap_WorldGroupSettings
   samplerType=SAMPLERTYPE_LinearColor
   texture=&DFHeightMask
  }
  n94 = LinearInterpolate(A=Constant() {r=-1}, B=Constant() {r=1}, Alpha=@n93.G) {constA=-1}
  n95 = ScalarParameter() {defaultValue=0.5, group=Cube_OG, parameterName=Rune_Intensity}
  n96 = TextureSampleParameter2D(UVs=BumpOffset(
    Coordinate=@n92.Result
    Height=Multiply(A=@n94, B=@n95)
    HeightRatioInput=ScalarParameter() {defaultValue=0.15, group=Cube_OG, parameterName=Rune_HeightRatio}
   )) {
   group=Cube_OG
   parameterName=RuneTex
   samplerSource=SSM_Wrap_WorldGroupSettings
   samplerType=SAMPLERTYPE_LinearColor
   texture=&DFHeightMask
  }
  n97 = ScalarParameter() {defaultValue=1, group=Cube_OG, parameterName=RunePatternAlpha}
  n98 = LinearInterpolate(
   A=Power(Base=@n85.G, Exp=LinearInterpolate(A=@n86, B=@n87, Alpha=@n88))
   B=Power(Base=@n96.G, Exp=ScalarParameter() {defaultValue=20, group=Cube_OG, parameterName=RunePower})
   Alpha=Saturate(None=Multiply(A=@n96.G, B=@n97))
  )
  n99 = Add(
   A=Multiply(A=ScalarParameter() {defaultValue=32, group=Cube_OG, parameterName=RuneTexTiling01}, B=TextureCoordinate())
   B=NamedRerouteUsage()
  )
  n100 = ScalarParameter() {defaultValue=-1, group=Cube_OG, parameterName=RuneHeightMin}
  n101 = ScalarParameter() {defaultValue=1, group=Cube_OG, parameterName=RuneHeightMax}
  n102 = TextureSampleParameter2D(UVs=@n99) {
   group=Cube_OG
   parameterName=RuneTex
   samplerSource=SSM_Wrap_WorldGroupSettings
   samplerType=SAMPLERTYPE_LinearColor
   texture=&DFHeightMask
  }
  n103 = TextureSampleParameter2D(UVs=BumpOffset(
    Coordinate=@n99
    Height=LinearInterpolate(A=@n100, B=@n101, Alpha=@n102.R)
    HeightRatioInput=ScalarParameter() {defaultValue=0.06, group=Cube_OG, parameterName=RuneRatio01}
   )) {
   group=Cube_OG
   parameterName=RuneTex
   samplerSource=SSM_Wrap_WorldGroupSettings
   samplerType=SAMPLERTYPE_LinearColor
   texture=&DFHeightMask
  }
  n104 = ScalarParameter() {defaultValue=1.5, parameterName=G_6}
  n105 = ScalarParameter() {defaultValue=15, parameterName=G_1}
  n106 = Multiply(A=Power(Base=@n103.R, Exp=LinearInterpolate(A=@n104, B=@n105, Alpha=@n88)), B=ScalarParameter() {defaultValue=5, parameterName=G_9})
  n107 = Multiply(A=ScalarParameter() {defaultValue=2, group=Cube_OG, parameterName=Pattern_Tiling}, B=TextureCoordinate())
  n108 = NamedRerouteUsage()
  n109 = ScalarParameter() {defaultValue=0.0135, group=Cube_OG, parameterName=Pattern_TimeSpeed}
  n110 = Time()
  n111 = TextureSampleParameter2D(UVs=Panner(Coordinate=Add(A=@n107, B=@n108), Time=Multiply(A=@n109, B=@n110)) {speedX=1}) {
   group=Cube_OG
   parameterName=RuneTex
   samplerSource=SSM_Wrap_WorldGroupSettings
   samplerType=SAMPLERTYPE_LinearColor
   texture=&DFHeightMask
  }
  n112 = Saturate(None=LinearInterpolate(
    A=ScalarParameter() {defaultValue=1, group=Cube_OG, parameterName=MinGrid}
    B=ScalarParameter() {defaultValue=-1, group=Cube_OG, parameterName=MaxGrid}
    Alpha=@n111.B
   ))
  n113 = Saturate(None=Multiply(A=Multiply(A=@n103.R, B=@n112), B=ScalarParameter() {defaultValue=1.25, parameterName=G_3}))
  n114 = Multiply(
   A=LinearInterpolate(
    A=ScalarParameter() {defaultValue=-0.1, parameterName=deactivated}
    B=ScalarParameter() {defaultValue=1, parameterName=activated}
    Alpha=NamedRerouteUsage()
   )
   B=Multiply(A=@n77.RGB, B=LinearInterpolate(A=@n98, B=@n106, Alpha=@n113))
  )
  n115 = Add(A=Multiply(A=@n70, B=@n74), B=Multiply(A=Power(Base=@n75, Exp=@n76), B=Multiply(A=@n74, B=@n114)))
  n116 = NamedRerouteUsage()
  n117 = LinearInterpolate(
   A=Multiply(A=ScalarParameter() {defaultValue=0.18, parameterName=InnerGlowWhileOff}, B=@n115)
   B=Multiply(A=@n115, B=Constant() {r=1})
   Alpha=@n116
  )
  n118 = NamedRerouteUsage()
  n119 = NamedRerouteUsage()
  n120 = Add(A=@n70, B=Multiply(A=@n70, B=Multiply(A=@n119, B=@n118)))
  n121 = Multiply(A=ScalarParameter() {defaultValue=0.1, parameterName=LowQ-InnerGlowWhileOff}, B=@n120)
  n122 = Multiply(A=@n120, B=ScalarParameter() {defaultValue=0.4, parameterName=LowQ-InnerGlowWhileOn})
  NamedRerouteDeclaration(None=QualitySwitch(Default=@n117, Low=LinearInterpolate(A=@n121, B=@n122, Alpha=@n116), High=@n117)) {name=CubeEM, nodeColor=<rgba 1 0.8906 0 1>}
  n123 = ScalarParameter() {parameterName=RuneActive}
  NamedRerouteDeclaration(None=@n123) {name=RuneActive, nodeColor=<rgba 0 1 1 1>}
  n124 = Time()
  n125 = ScalarParameter() {defaultValue=0.5, group=Cube_OG, parameterName=PulseSpeed_3}
  NamedRerouteDeclaration(None=MaterialFunctionCall(Value=Multiply(A=@n124, B=@n125)) {materialFunction=&LinearSine}."Rounded Linear Sine ") {name=AnimatedSine, nodeColor=<rgba 0.1484 1 0 1>}
  NamedRerouteDeclaration(None=Multiply(A=@n106, B=@n113)) {name=SinePattern, nodeColor=<rgba 0.7812 0 1 1>}
  NamedRerouteDeclaration(None=@n59) {name=SphereMaskUV01, nodeColor=<rgba 1 0 0.9844 1>}
  n126 = Multiply(A=ScalarParameter() {group=Cube_OG, parameterName=Zap}) {constB=2}
  n127 = Multiply(A=ScalarParameter() {group=Cube_OG, parameterName=Zap2}) {constB=1.5}
  NamedRerouteDeclaration(None=Add(A=Add(A=@n126, B=@n127), B=@n123)) {name=RuneActiveMod, nodeColor=<rgba 0 1 0.6953 1>}
  NamedRerouteDeclaration(None=@n77.RGB) {name=GlowColor, nodeColor=<rgba 0.2656 1 0 1>}
  n128 = Saturate(None=Add(A=Fresnel() {baseReflectFraction=0}, B=ScalarParameter() {defaultValue=1, group=Cube_OG, parameterName=MetalExp}))
  n129 = ScalarParameter() {defaultValue=0.2, group=Cube_OG, parameterName=Metal}
  NamedRerouteDeclaration(None=StaticSwitchParameter(True=Multiply(A=@n128, B=@n129), False=Constant() {r=0.2}) {group=Cube_OG, parameterName=UseAdjustableMetal}) {name=Metalness, nodeColor=<rgba 0 1 0.7422 1>}
  n130 = MaterialFunctionCall(UVs=TextureCoordinate() {coordinateIndex=1}, RoomZscale=Constant() {r=0.1}, Tiling=Constant2Vector() {g=1, r=1}) {materialFunction=&MF_InteriorCubemapUV}
  NamedRerouteDeclaration(None=Saturate(None=ComponentMask(None=@n130.UVW) {r=true})) {name=CenterMask, nodeColor=<rgba 1 0.9141 0 1>}
  NamedRerouteDeclaration(None=@n1) {name=Material_OG, nodeColor=<rgba 1 0.375 0 1>}
  n131 = StaticSwitchParameter(True=@n17.Result, False=Constant()) {group=Cube_Additional_Control, parameterName=UseHuntGlitch}
  NamedRerouteDeclaration(None=Multiply(A=SmoothStep(Value=@n131) {constMax=0.5}, B=Constant2Vector() {g=0.2, r=0.2}) {constB=0.1}) {name=Distortion, nodeColor=<rgba 1 0 0.9375 1>}
  n132 = TextureSampleParameter2D(UVs=NamedRerouteUsage()) {
   group=Cube_OG
   parameterName=RuneTex
   samplerSource=SSM_Wrap_WorldGroupSettings
   samplerType=SAMPLERTYPE_LinearColor
   texture=&DFHeightMask
  }
  n133 = ScalarParameter() {defaultValue=20, group=Cube_OG, parameterName=RunePower}
  n134 = VectorParameter() {group=Cube_Additional, parameterName=InnerIntensity, defaultValue=<rgba 0 0.7606 1 5>}
  n135 = Multiply(A=@n134.RGB, B=@n134.RGB)
  n136 = OneMinus(None=NamedRerouteUsage())
  NamedRerouteDeclaration(None=Multiply(A=Power(Base=@n132.R, Exp=@n133), B=Multiply(A=@n135, B=@n136))) {name=InnerPattern, nodeColor=<rgba 1 0 0.5859 1>}
  n137 = TextureObjectParameter() {
   group=Cube_OG
   parameterName=CubeNoiseTexture
   samplerType=SAMPLERTYPE_Masks
   texture=&/T_CubeNoise_Character
  }
  NamedRerouteDeclaration(None=@n137) {name=CubeNoise_Texture, nodeColor=<rgba 0.7109 0 1 1>}
  n138 = Multiply(A=ScalarParameter() {defaultValue=2, group=Cube_OG, parameterName=MasterNoiseSCale}, B=TextureCoordinate())
  n139 = TextureSample(UVs=@n138, Tex=NamedRerouteUsage()) {
   samplerSource=SSM_Wrap_WorldGroupSettings
   samplerType=SAMPLERTYPE_Masks
   texture=&/T_CubeNoise
  }
  NamedRerouteDeclaration(None=@n139.R) {name=CubeDefaultNoise, nodeColor=<rgba 0 1 0.7891 1>}
  n140 = TextureSample!(
   UVs=BumpOffset(Coordinate=@n138, Height=Constant() {r=0.58}, HeightRatioInput=Constant() {r=-0.5})
   Tex
   Bias=ScalarParameter() {defaultValue=4, group=Cube_OG, parameterName=Mipbias_Shadow}
   Apply View MipBias
  ) {mipValueMode=TMVM_MipBias, samplerType=SAMPLERTYPE_Masks, texture=&T_CubeNoise}
  NamedRerouteDeclaration(None=@n140.R) {name=BlurredShadow, nodeColor=<rgba 1 0 0.9844 1>}
  NamedRerouteDeclaration(None=@n140.G) {name=BlurredNoise, nodeColor=<rgba 0.4531 0 1 1>}
  n141 = VectorParameter() {group=Cube_OG, parameterName=SparklesColor, defaultValue=<rgba 0.5857 0.2 1 1>}
  n142 = ScalarParameter() {defaultValue=1, group=Cube_OG, parameterName=StarsTileMod}
  n143 = TextureCoordinate()
  n144 = Multiply(A=ScalarParameter() {defaultValue=6, group=Cube_OG, parameterName=SparkleTiling01}, B=@n143)
  n145 = Constant() {r=-2}
  n146 = Constant() {r=0.02}
  n147 = NamedRerouteUsage()
  n148 = TextureSample!(UVs=Multiply(A=@n142, B=BumpOffset(Coordinate=@n144, Height=@n145, HeightRatioInput=@n146)), Tex=@n147, Apply View MipBias) {
   samplerSource=SSM_Wrap_WorldGroupSettings
   samplerType=SAMPLERTYPE_Masks
   texture=&T_CubeNoise
  }
  n149 = Multiply(A=@n143, B=ScalarParameter() {defaultValue=12, parameterName=SparkleTiling02})
  n150 = Constant() {r=0.2}
  n151 = TextureSample(UVs=Multiply(A=@n142, B=BumpOffset(Coordinate=@n149, Height=@n145, HeightRatioInput=@n150)), Tex=@n147) {
   samplerSource=SSM_Wrap_WorldGroupSettings
   samplerType=SAMPLERTYPE_Masks
   texture=&T_CubeNoise
  }
  n152 = Multiply(A=@n148.B, B=@n151.B)
  n153 = LinearInterpolate(A=Constant() {r=7}, B=Constant() {r=8}, Alpha=NamedRerouteUsage())
  n154 = NamedRerouteDeclaration(None=Multiply(A=Power(Base=@n152, Exp=@n153), B=ScalarParameter() {defaultValue=40, group=Cube_OG, parameterName=SparkleIntensity})) {name=Sparkles, nodeColor=<rgba 0 0.5078 1 1>}
  n155 = Add(A=Power(Base=NamedRerouteUsage(), Exp=Constant() {r=3}), B=Constant() {r=-0.05})
  NamedRerouteDeclaration(None=Multiply(A=Multiply(A=@n141.RGB, B=@n154), B=Saturate(None=@n155))) {name=StarColor, nodeColor=<rgba 1 0.4453 0 1>}
  n156 = Saturate(None=Multiply(A=@n33, B=Constant() {r=2}))
  n157 = Constant() {r=1.8}
  n158 = Add(
   A=MaterialFunctionCall(Transition Phase=Power(Base=@n156, Exp=@n157), Height Texture=@n39) {materialFunction=&HeightLerp}.Lerp Alpha No Contrast
   B=@n39
  )
  NamedRerouteDeclaration(None=LinearInterpolate(A=Constant() {r=0.2}, B=Constant() {r=0.1}, Alpha=Saturate(None=@n158))) {name=Spec, nodeColor=<rgba 0.8281 0 1 1>}
  NamedRerouteUsage()
  ScalarParameter() {defaultValue=1, parameterName=Param_1}
 ]}
