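# Masked, two-sided, Default Lit material for cube cosmetics, previewed on the Quest_DarkNexusShape_A prop mesh.
# It wraps the SpacePlungeCube function, at a constant full mask, between the 2023 CharacterParent PreFX and PostFX functions, using material attributes.
# PreFX feeds the effect's input and also passes side data directly to PostFX.
# Usage flags cover skeletal meshes, instanced static meshes and mesh deformers.
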Material /Hera/Props/Quest/Quest_DarkNexusShape_A/Material/M_Cube_Cosmetics
graph {
 properties {
  bUseMaterialAttributes=true
  bUsedWithInstancedStaticMeshes=true
  bUsedWithMeshDeformer=true
  bUsedWithSkeletalMesh=true
  blendMode=BLEND_Masked
  pixelDepthOffsetMode=PDOM_Legacy
  previewMesh=/Hera/Props/Quest/Quest_DarkNexusShape_A/Meshes/SM_Quest_DarkNexusShape_A
  twoSided=true
  usedShadingModels=Default Lit
 }
 nodes [
  n1 = MaterialFunctionCall() {materialFunction=&MF_CharacterParent_2023_PreFX}
  MaterialFunctionCall(
   Material=MaterialFunctionCall!(In=@n1.Result, Mask=Constant() {r=1}) {materialFunction=&/MF_SpacePlungeCube}.Result
   M=@n1.M
   Pre Skinned Local Position=@n1.M
   Distance Blend=@n1.M
   Fresnel=@n1.M
  ) {materialFunction=&MF_CharacterParent_2023_PostFX}
 ]
}
